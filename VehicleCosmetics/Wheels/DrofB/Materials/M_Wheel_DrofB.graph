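# Surface material for the DrofB cosmetic vehicle wheel, with an animation-ready skeletal mesh usage and a tintable glow.
# Vertex color red blends rim and tire normals and glow masks within one material.
# Emissive adds masked glow to parallax-offset swirl noise, shaped by a view-angle band from the blended normal, both tinted by CustomColor and GlowBrightness.
# Attributes then pass through shared vehicle damage and de-rez transition functions.
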Material /VehicleCosmetics/Wheels/DrofB/Materials/M_Wheel_DrofB
graph {
 properties {
  bAutomaticallySetUsageInEditor=false
  bUseMaterialAttributes=true
  bUsedWithMeshDeformer=true
  bUsedWithSkeletalMesh=true
  previewMesh=/VehicleCosmetics/Wheels/DrofB/Meshes/Wheel_DROFB_SK
  usedShadingModels=Default Lit
 }
 nodes [
  n1 = SetMaterialAttributes!(
   MaterialAttributes
   Emissive Color=NamedRerouteUsage()
   Normal=NamedRerouteUsage()
   Base Color=VectorParameter() {group=00_Base, parameterName=Base Color}.RGB
   Metallic=ScalarParameter() {defaultValue=0.5, group=00_Base, parameterName=Metallic}
   Roughness=ScalarParameter() {defaultValue=0.2, group=00_Base, parameterName=Roughness}
  ) {attributeSetTypes [
    #1
    #2
    #3
    #4
    #5
   ]}
  GetMaterialAttributes(None=MaterialFunctionCall(In=MaterialFunctionCall(Material=@n1) {materialFunction=&MF_BaseVehicle_Damage}.Result) {materialFunction=&MF_VehicleDeRez_Transition}.Result) {attributeGetTypes [#6]}
  n2 = BumpOffset(Coordinate=TextureCoordinate() {uTiling=0.1, vTiling=0.1}, Height=Constant() {r=1}) {heightRatio=0.5}
  n3 = Multiply(
   A=VectorParameter() {group=00_Base, parameterName=CustomColor, defaultValue=<rgba 0 0.3231 1 1>}.RGB
   B=ScalarParameter() {defaultValue=3, group=00_Base, parameterName=GlowBrightness}
  )
  n4 = Multiply(A=Multiply(
    A=TextureSample(UVs=@n2) {samplerSource=SSM_Wrap_WorldGroupSettings, texture=/VehicleCosmetics/Wheels/_SharedTextures/FX/Noise_Swirls_01_Pack}.G
    B=@n3
   )) {constB=4}
  n5 = TextureSampleParameter2D() {
   group=00_Base
   parameterName=RimNormal
   samplerSource=SSM_Wrap_WorldGroupSettings
   samplerType=SAMPLERTYPE_Normal
   texture=/VehicleCosmetics/Wheels/DrofB/Textures/T_DrofB_Rim_N
  }
  n6 = TextureSampleParameter2D() {
   group=00_Base
   parameterName=TireNormal
   samplerSource=SSM_Wrap_WorldGroupSettings
   samplerType=SAMPLERTYPE_Normal
   texture=/VehicleCosmetics/Wheels/DrofB/Textures/T_DrofB_Tire_N
  }
  n7 = VertexColor()
  n8 = LinearInterpolate(A=@n5.RGB, B=@n6.RGB, Alpha=@n7.R)
  n9 = DotProduct(A=Transform(None=@n8), B=CameraVectorWS())
  n10 = OneMinus(None=@n9)
  n11 = Multiply(A=Clamp(None=Multiply(A=@n9) {constB=2}) {minDefault=0.35}, B=Clamp(None=ConstantBiasScale(None=@n10) {bias=-0.375, scale=1.666}))
  n12 = Multiply(A=@n11, B=@n11)
  n13 = TextureSampleParameter2D() {
   group=00_Base
   parameterName=RimMasks
   samplerSource=SSM_Wrap_WorldGroupSettings
   samplerType=SAMPLERTYPE_LinearColor
   texture=/VehicleCosmetics/Wheels/DrofB/Textures/T_DrofB_Rim_RGB
  }
  n14 = TextureSampleParameter2D() {
   group=00_Base
   parameterName=TireMasks
   samplerSource=SSM_Wrap_WorldGroupSettings
   samplerType=SAMPLERTYPE_LinearColor
   texture=/VehicleCosmetics/Wheels/DrofB/Textures/T_Drof_Tire_M
  }
  n15 = LinearInterpolate(A=Multiply(A=@n13.R, B=@n13.R), B=@n14.R, Alpha=@n7.R)
  NamedRerouteDeclaration(None=Add(A=Multiply(A=@n4, B=@n12), B=Multiply(A=@n3, B=@n15))) {name=Emissive, nodeColor=<rgba 0 1 0.9853 1>}
  NamedRerouteDeclaration(None=@n8) {name=Normal, nodeColor=<rgba 0.5 0.0417 0.3546 1>}
 ]
}
