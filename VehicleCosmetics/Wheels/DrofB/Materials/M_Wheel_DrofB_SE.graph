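# Translucent, unlit, two-sided material for the DrofB cosmetic wheel, previewed on its Holo mesh, that renders a glowing animated caustics effect.
# Two samples of a caustics texture pan at different speeds over V-stretched UVs, timed by PanSpeedAdjust.
# One sample plus the other squared is masked by a shockwave curve texture, with a faint mask-only base, to drive opacity.
# Emissive is a flat CustomColor scaled by EmissiveMultiply, so all visible patterning comes from opacity.
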Material /VehicleCosmetics/Wheels/DrofB/Materials/M_Wheel_DrofB_SE
graph {
 properties {
  bAutomaticallySetUsageInEditor=false
  bUseMaterialAttributes=true
  bUsedWithMeshDeformer=true
  bUsedWithSkeletalMesh=true
  blendMode=BLEND_Translucent
  previewMesh=/VehicleCosmetics/Wheels/DrofB/Meshes/Wheel_DROFB_Holo
  shadingModel=MSM_Unlit
  twoSided=true
  usedShadingModels=Unlit
 }
 nodes [
  n1 = TextureSample() {texture=/VehicleCosmetics/Wheels/_SharedTextures/FX/Curve_Shockwaves_Pack}
  n2 = TextureCoordinate() {vTiling=0.1}
  n3 = Multiply(A=ScalarParameter() {defaultValue=1, parameterName=PanSpeedAdjust}, B=Time())
  n4 = TextureSample(UVs=Panner(Coordinate=@n2, Time=@n3) {speedX=0.075, speedY=0.0675}) {texture=/VehicleCosmetics/Wheels/_SharedTextures/FX/Caustics_D}
  Multiply(A=Multiply(A=@n1.R, B=@n4.R)) {constB=32}
  n5 = Add(
   A=TextureSample(UVs=Panner(Coordinate=@n2, Time=@n3) {speedX=0.125, speedY=1}) {texture=&Caustics_D}.R
   B=Multiply(A=@n4.R, B=@n4.R)
  )
  SetMaterialAttributes!(
   MaterialAttributes
   Emissive Color=Multiply(
    A=VectorParameter() {parameterName=CustomColor, defaultValue=<rgba 0.1878 0.8308 0.6584 1>}.RGB
    B=ScalarParameter() {defaultValue=32, parameterName=EmissiveMultiply}
   )
   Opacity=Add(A=Multiply(A=@n5, B=@n1.R), B=Multiply(A=@n1.R) {constB=0.005})
   Refraction
  ) {attributeSetTypes [#1, #2, #3]}
 ]
}
